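; Gameplay ability for kill-and-collect quests that spawns collectible items when a valid tandem NPC or wildlife pawn is killed.
; The kill event's target is matched by gameplay tag to a lookup class listing its item classes.
; Each item's random count uses scalable-float quest limits, falling back to class defaults when either is -1.
; Items chain outward at random XY offsets, starting at the killed pawn, then from the previous spawn.
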(event Ability|EventActivateAbilityFromEvent
  (params (EventData :type Gameplay%20Event%20Data%20Structure%20%28by%20ref%29))
  (bind (_eventtag _instigator _target) (Utilities|Struct|BreakGameplayEventData EventData))
  ;[1 1 1] Activation from kill event, is target an NPC (of FortPlayerPawnAthena class) or Irwin?
  (|CallParentFunction EventData)
  (bind _asfort_player_pawn_athena (Utilities|Casting|CastToFortPlayerPawnAthena _target)
    (:then
      (bind _returnvalue (GameplayTags|HasAnyMatchingGameplayTags(Message) _target (Variables|Default|GetValidTandemTags))) ; Make sure this Tandem is valid (i.e. CubeTheif or Boss), not a player or otherwise
      (if _returnvalue
        (Variables|Default|SetPawnKilled _asfort_player_pawn_athena)
        (else
          (Ability|EndAbility))))
    (:CastFailed
      (bind _asnpc_pawn_wildlife_parent (Utilities|Casting|CastToNPC_Pawn_Wildlife_Parent _target)
        (:then
          (Variables|Default|SetPawnKilled _asnpc_pawn_wildlife_parent))
        (:CastFailed
          (Ability|EndAbility))) ; Otherwise, check if pawn is Irwin
    ))
  ;[1 1 1] Loops through valid tags to check which pawn type was damaged (i.e. Irwin.Burt, CubeThief, etc.)
  (bind _keys (Utilities|Map|Keys (Variables|Default|GetPawnTagsToChildClassLookups)))
  (bind _output_get (Variables|Default|SetT_ValidPawnTags _keys)) ; Get valid pawn tags from map keys
  (bind _array_element (Utilities|Array|ForEachLoopwithBreak _output_get)
    (:LoopBody
      (if (GameplayTags|HasTag (GameplayTags|GetOwnedGameplayTags (Variables|Default|GetPawnKilled) :!class /Script/GameplayTags.GameplayTagAssetInterface) _array_element true)
        (Variables|Default|SetT_PawnType _array_element)))
    (:Completed
      _))
  (CallFunction|SpawnItem))

(event Custom|SpawnItem
  (bind _t_pawntype (Variables|Default|GetT_PawnType))
  (bind _pawntagstochildclasslookups (Variables|Default|GetPawnTagsToChildClassLookups))
  (bind _value (Utilities|Map|Find _pawntagstochildclasslookups _t_pawntype))
  ;[1 1 1] If both tag and class found are valid and exist in the lookup, spawn the item for that Pawn type
  (if (GameplayTags|IsGameplayTagValid (Variables|Default|GetT_PawnType))
    (if (Utilities|IsValidClass _value)
      ;;[0 0 0 0.6] For each item that this pawn type can spawn, handle the spawning logic independently
      (for _array_element (Class|GetClassDefaults _value)
        (if (Utilities|IsValidClass _array_element)
          (bind _randomspawncount (CallFunction|GetRandomSpawnCount _array_element))
          (if (> _randomspawncount 0)
            ;;;[0 0 0 0.8] NOTE: Exponential increase in complexity for more than one type of item that needs to be spawned
            (for _index (range 1 (+ _randomspawncount 1))
              (if (== 1 _index)
                (bind _randomtransform (CallFunction|MakeRandomXYTransform (Transformation|GetWorldLocation (Variables|/Script/Engine.Character|GetCapsuleComponent (Variables|Default|GetPawnKilled)))))
                (bind _returnvalue (Building|SpawnBuildingGameplayActor _array_element _randomtransform (Variables|Default|GetPawnKilled)))
                (Variables|Default|SetPrevSpawnedItem _returnvalue)
                (else
                  (bind _randomtransform_2 (CallFunction|MakeRandomXYTransform (Transformation|GetActorLocation (Variables|Default|GetPrevSpawnedItem))))
                  (bind _returnvalue_2 (Building|SpawnBuildingGameplayActor _array_element _randomtransform_2 (Variables|Default|GetPawnKilled)))
                  (Variables|Default|SetPrevSpawnedItem _returnvalue_2)))))
          (else
            (Development|PrintString (Utilities|String|Append (Utilities|GetClassDisplayName _value) "'s array contains an invalid or empty child class reference.") true true "(R=1.000000,G=0.000000,B=0.534290,A=1.000000)"))))
      (else
        (Development|PrintString "Pawn lookup child class is not valid or was not found in the map" true true "(R=1.000000,G=0.000000,B=0.410990,A=1.000000)")))
    (Ability|EndAbility)
    (else
      (Development|PrintString "Pawn tag is not valid or was not found in the map" true true "(R=1.000000,G=0.000000,B=0.250942,A=1.000000)")
      (Ability|EndAbility))))

(fn GetRandomSpawnCount
  (params (ChildClassRef :type BP%20Athena%20Kill%20And%20Collect%20Parent%20Class%20Reference))
  (bind _maxdrop (LocalVariables|GetMaxDrop))
  (bind _returnvalue (ScalableFloat|GetValueatLevel _maxdrop))
  (bind _returnvalue_2 (Math|Float|Floor _returnvalue))
  (bind _mindrop (LocalVariables|GetMinDrop))
  (bind _returnvalue_3 (ScalableFloat|GetValueatLevel _mindrop))
  (bind _returnvalue_4 (Math|Float|Floor _returnvalue_3))
  (bind (_sm_droppeditem _fortplayercontroller _meshrelativescale _ambientniagarasystem _ambientparticle _ambientfx_transform _ambientsound _pickupsound _ambientaudiofadeduration _ambientcollisionvolumeradius _mindrop_2 _maxdrop_2) (Class|GetClassDefaults ChildClassRef))
  ;[1 1 1] Takes min and max range for specified type from either the child class blueprints or AthenaQuestData table (if not -1) to determine count of the item to be spawned on this execution of the ability
  (if (and (!= _returnvalue_4 -1) (!= _returnvalue_2 -1))
    (LocalVariables|SetMinDropLocal _returnvalue_4)
    (LocalVariables|SetMaxDropLocal _returnvalue_2)
    (else
      (LocalVariables|SetMinDropLocal _mindrop_2)
      (LocalVariables|SetMaxDropLocal _maxdrop_2)))
  (if (and (>= (LocalVariables|GetMinDropLocal) 0) (<= (LocalVariables|GetMinDropLocal) (LocalVariables|GetMaxDropLocal)))
    (if (> (LocalVariables|GetMaxDropLocal) 0)
      (return (Math|Random|RandomIntegerinRange (LocalVariables|GetMinDropLocal) (LocalVariables|GetMaxDropLocal)))
      (else
        (Development|PrintString "Invalid max spawn count, or max spawn count is 0. Item will not spawn" true true "(R=1.000000,G=0.000000,B=0.452663,A=1.000000)")))
    (else
      (Development|PrintString "Invalid min spawn count! Item will not spawn" true true "(R=1.000000,G=0.000000,B=0.349378,A=1.000000)")))
  (return 0))

(fn ShouldAbilityRespondToEvent
  (params (ActorInfo :type Gameplay%20Ability%20Actor%20Info%20Structure) (Payload :type Gameplay%20Event%20Data%20Structure))
  (return true))

(fn MakeRandomXYTransform
  (params (StartingSpawnLocation :type Vector))
  ;[1 1 1] Spawns the item(s) in a random x-y axis arrangement from the location given (first will be from where the pawn is killed, subsequent from where the previous item spawned)
  (bind _output_get (LocalVariables|SetAdjustedVectorLocal (Utilities|MakeRandomVector StartingSpawnLocation 3e+01 6e+01)))
  (return (Math|Transform|MakeTransform (Math|Vector|MakeVector (.x _output_get) (.y _output_get) (.z StartingSpawnLocation)))))
